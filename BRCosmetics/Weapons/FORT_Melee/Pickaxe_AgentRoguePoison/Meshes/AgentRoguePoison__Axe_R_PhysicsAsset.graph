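# Physics asset for the right-hand axe mesh of the AgentRoguePoison pickaxe, a cosmetic melee weapon.
# It has a single body on the Melee_TwoHanded_Handle bone, shaped as one offset, rotated capsule.
# It defines no constraints, so the mesh has one simple collision shape and no jointed parts.
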PhysicsAsset /BRCosmetics/Weapons/FORT_Melee/Pickaxe_AgentRoguePoison/Meshes/AgentRoguePoison__Axe_R_PhysicsAsset
bodies {Melee_TwoHanded_Handle [{shapeType=Capsule, length=35.4029, radius=6.587, center=<0.0025 -1.6853 22.5256>, rotation=<rot 1.0281 -90 0.0106>}]}
constraints []
